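# Material function for the BisonDrain backpack's reactive TV screen, outputting lit and tinted shaded colors.
# A phase from StartTime and PlayTime, or FrontendAlpha in the front end, steps through four chained flipbooks with a scanline effect.
# Panning masks build colored glitch lines, UV distortion and black flicker, while layered noise adds static, with distance blends adjusting intensity.
# A final lerp, gated by a step or ReactivityEnable, selects between screen states.
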MaterialFunction /BRCosmetics/Accessories/FORT_Backpacks/Backpack_M_MED_BisonDrain/Materials/MF_BisonTV_Reactivity
graph {nodes [
  n1 = NamedRerouteUsage()
  n2 = LinearInterpolate(
   A=NamedRerouteUsage()
   B=Add(A=NamedRerouteUsage(), B=NamedRerouteUsage())
   Alpha=LinearInterpolate(A=Step(Y=@n1) {constX=0.99}, B=ScalarParameter() {group=Reactivity, parameterName=ReactivityEnable}, Alpha=NamedRerouteUsage())
  )
  FunctionOutput(None=@n2) {outputName=TV_Lit}
  FunctionOutput(None=Multiply(A=@n2, B=VectorParameter() {group=Default_Looping, parameterName=ShadedTint, defaultValue=<rgba 0.6906 0.6906 0.6906 1>}.RGB)) {outputName=TV_Shaded, sortPriority=1}
  n3 = Multiply(A=NamedRerouteUsage(), B=Constant2Vector() {g=1, r=2}) {constB=1.5}
  n4 = ScalarParameter() {defaultValue=0.2, group=TV Glitches, parameterName=SpeedX04}
  n5 = Constant() {r=2}
  n6 = TextureObjectParameter() {
   group=TV Glitches
   parameterName=ShortLineMask
   samplerType=SAMPLERTYPE_Masks
   texture=/BRCosmetics/Accessories/FORT_Backpacks/Backpack_M_MED_BisonDrain/Textures/T_BisonDrain_SmallLineMask
  }
  n7 = TextureSample(UVs=Panner(Coordinate=@n3, Speed=AppendVector(A=@n4, B=@n5)), Tex=@n6) {samplerType=SAMPLERTYPE_Masks, texture=/BRCosmetics/Accessories/FORT_Backpacks/Backpack_M_MED_BisonDrain/Textures/T_BisonDrain_LineMask}
  n8 = LinearInterpolate(
   A=Constant3Vector() {constant=<rgba 0.8125 0.8125 0.8125 0>}
   B=VectorParameter() {group=TV Glitches, parameterName=GreenLine, defaultValue=<rgba 0.0219 1 0.568 1>}.RGB
   Alpha=@n7.R
  )
  n9 = VectorParameter() {group=TV Glitches, parameterName=PinkLine, defaultValue=<rgba 1 0.3643 0.7462 1>}
  NamedRerouteDeclaration(None=LinearInterpolate(
    A=LinearInterpolate(A=@n8, B=@n9.RGB, Alpha=@n7.G)
    B=VectorParameter() {group=TV Glitches, parameterName=PurpleLine, defaultValue=<rgba 0.5204 0.2053 1 1>}.RGB
    Alpha=@n7.B
   )) {name=LineBackground, nodeColor=<rgba 0 0.1328 1 1>}
  n10 = MaterialFunctionCall(Float2=NamedRerouteUsage()) {materialFunction=&BreakOutFloat2Components}
  n11 = Constant() {r=1920}
  n12 = Constant() {r=1080}
  n13 = MaterialFunctionCall(X=Divide(A=@n10.R, B=Divide(A=@n11, B=@n12)), Y=@n10.G) {materialFunction=&MakeFloat2}
  n14 = ScalarParameter() {defaultValue=1, group=Default_Looping, parameterName=DefaultTilling}
  n15 = VectorParameter() {group=Default_Looping, parameterName=DefaultOffset, defaultValue=<rgba 0 0 0 1>}
  NamedRerouteDeclaration(None=Add(A=Multiply(A=@n13.Result, B=@n14), B=ComponentMask(None=@n15.RGB) {g=true, r=true})) {name=TV_UV, nodeColor=<rgba 0 0.7422 1 1>}
  n16 = VectorParameter() {group=Default_Looping, parameterName=ConstantNoise_Color, channelNames {a=Intensity, b="", g="", r=""}, defaultValue=<rgba 1 1 1 1>}
  n17 = Multiply(A=NamedRerouteUsage(), B=ScalarParameter() {defaultValue=1, group=Default_Looping, parameterName=TVNoiseTiling})
  n18 = TextureSample(UVs=Panner(Coordinate=@n17) {speedX=2, speedY=0.1}) {
   samplerSource=SSM_Wrap_WorldGroupSettings
   samplerType=SAMPLERTYPE_Masks
   texture=/Game/Missions/Primary/RepairTheShelter/Textures/Cinematic_Screen_Static
  }
  n19 = TextureSample(UVs=Panner(Coordinate=@n17) {speedX=2, speedY=-0.05}) {
   samplerSource=SSM_Wrap_WorldGroupSettings
   samplerType=SAMPLERTYPE_Masks
   texture=&Cinematic_Screen_Static
  }
  n20 = TextureSample(UVs=Panner(Coordinate=@n17) {speedX=-1, speedY=-0.03}) {
   samplerSource=SSM_Wrap_WorldGroupSettings
   samplerType=SAMPLERTYPE_Masks
   texture=&Cinematic_Screen_Static
  }
  n21 = Add(A=Multiply(A=@n18.B) {constB=0.2}, B=Multiply(A=@n19.G, B=@n20.R))
  n22 = LinearInterpolate(
   A=ScalarParameter() {defaultValue=3, group=TV Glitches, parameterName=Static Scale}
   B=ScalarParameter() {defaultValue=40, group=TV Glitches, parameterName=Static Scale_Distant}
   Alpha=MaterialFunctionCall!(
    Blend Range=ScalarParameter() {defaultValue=200, group=Reactivity, parameterName=BlendRange}
    Start Offset=ScalarParameter() {defaultValue=100, group=Reactivity, parameterName=StartOffset}
   ) {materialFunction=/Engine/Functions/Engine_MaterialFunctions03/Particles/Distance_Blend}.Result
  )
  n23 = Saturate(None=NamedRerouteDeclaration(None=Multiply(A=@n21, B=@n22)) {name=Static, nodeColor=<rgba 1 0.6562 0 1>})
  NamedRerouteDeclaration(None=Multiply(A=Multiply(A=@n16, B=@n16), B=Multiply(A=@n23) {constB=0.2, desc=constant static})) {name=StaticConstant, nodeColor=<rgba 0 0.7422 1 1>}
  n24 = NamedRerouteUsage()
  n25 = Subtract(A=@n24) {constB=0.25}
  n26 = Subtract(A=@n25) {constB=0.25}
  n27 = Saturate(None=@n26)
  NamedRerouteDeclaration(None=@n27) {name=PhaseSwitch02, nodeColor=<rgba 0 1 0.1797 1>}
  n28 = Saturate(None=Subtract(A=@n26) {constB=0.25})
  NamedRerouteDeclaration(None=@n28) {name=PhaseSwitch03, nodeColor=<rgba 0 1 0.1797 1>}
  n29 = Saturate(None=Multiply(A=@n24) {constB=4})
  n30 = Constant() {r=4}
  n31 = TextureObjectParameter() {
   group=Default_Looping
   parameterName=DefaultSerie01
   texture=/BRCosmetics/Accessories/FORT_Backpacks/Backpack_M_MED_BisonDrain/Textures/T_Flipbook_Intro01
  }
  n32 = NamedRerouteUsage()
  n33 = Saturate(None=@n25)
  n34 = Saturate(None=Multiply(A=@n33) {constB=4})
  n35 = Constant() {r=4}
  n36 = TextureObjectParameter() {
   group=Default_Looping
   parameterName=DefaultSerie02
   texture=/BRCosmetics/Accessories/FORT_Backpacks/Backpack_M_MED_BisonDrain/Textures/T_Intro02
  }
  n37 = NamedRerouteUsage()
  n38 = NamedRerouteUsage()
  n39 = Multiply(A=@n27) {constB=4}
  n40 = Constant() {r=4}
  n41 = TextureObjectParameter() {
   group=Default_Looping
   parameterName=DefaultSerie03
   texture=/BRCosmetics/Accessories/FORT_Backpacks/Backpack_M_MED_BisonDrain/Textures/T_Intro03
  }
  n42 = LinearInterpolate(
   A=LinearInterpolate(
    A=MaterialFunctionCall("Animation  Phase (0-1)"=@n29, Number of Rows=@n30, Number of Columns=@n30, Texture=@n31, UVs=@n32) {materialFunction=&FlipBook}.Result
    B=MaterialFunctionCall("Animation  Phase (0-1)"=@n34, Number of Rows=@n35, Number of Columns=@n35, Texture=@n36, UVs=@n37) {materialFunction=&FlipBook}.Result
    Alpha=Sign(None=@n38)
   )
   B=MaterialFunctionCall("Animation  Phase (0-1)"=Saturate(None=@n39), Number of Rows=@n40, Number of Columns=@n40, Texture=@n41, UVs=NamedRerouteUsage()) {materialFunction=&FlipBook}.Result
   Alpha=Sign(None=NamedRerouteUsage())
  )
  n43 = Constant() {r=4}
  n44 = TextureObjectParameter() {
   group=Default_Looping
   parameterName=DefaultSerie04
   texture=/BRCosmetics/Accessories/FORT_Backpacks/Backpack_M_MED_BisonDrain/Textures/T_Intro04
  }
  n45 = MaterialFunctionCall(
   "Animation  Phase (0-1)"=Saturate(None=Multiply(A=@n28) {constB=4})
   Number of Rows=@n43
   Number of Columns=@n43
   Texture=@n44
   UVs=NamedRerouteUsage()
  ) {materialFunction=&FlipBook}
  n46 = Sign(None=NamedRerouteUsage())
  n47 = Constant3Vector() {constant=<rgba 1 1 1 0>}
  NamedRerouteDeclaration(None=Multiply(
    A=LinearInterpolate(A=@n42, B=@n45.Result, Alpha=@n46)
    B=MaterialFunctionCall!(In_TechArtMask=@n47) {materialFunction=&/MF_BisonDrainScreen_Backpack}.ScanlineEffect
   )) {name=FlipbookAnimation, nodeColor=<rgba 0.7344 0 1 1>}
  n48 = NamedRerouteUsage()
  n49 = Multiply(A=NamedRerouteUsage()) {constB=0.1}
  n50 = ScalarParameter() {defaultValue=0.5, group=TV Glitches, parameterName=SpeedX01}
  n51 = Constant()
  n52 = TextureObjectParameter() {
   group=TV Glitches
   parameterName=LineMask
   samplerType=SAMPLERTYPE_LinearColor
   texture=&T_BisonDrain_LineMask
  }
  n53 = TextureSample(UVs=Panner(Coordinate=Subtract(A=@n48, B=@n49), Speed=AppendVector(A=@n50, B=@n51)), Tex=@n52) {
   samplerSource=SSM_Wrap_WorldGroupSettings
   samplerType=SAMPLERTYPE_LinearColor
   texture=&T_BisonDrain_LineMask
  }
  n54 = Frac(None=Multiply(A=ScalarParameter() {defaultValue=3, group=TV Glitches, parameterName=LerpSpeed}, B=Time()))
  n55 = LinearInterpolate(A=LinearInterpolate(A=@n53.R, B=@n53.R, Alpha=Step(X=@n54) {constY=0.3}), B=@n53.B, Alpha=Step(X=@n54) {constY=0.6})
  n56 = LinearInterpolate(
   A=Multiply(A=@n55, B=ScalarParameter() {defaultValue=1, group=TV Glitches, parameterName=BlackIntensity_Close})
   B=Multiply(A=@n55, B=ScalarParameter() {defaultValue=0.75, group=TV Glitches, parameterName=BlackIntensity_Far})
   Alpha=MaterialFunctionCall(Blend Range=Constant() {r=150}, Start Offset=Constant()) {materialFunction=&Distance_Blend}.Result
  )
  NamedRerouteDeclaration(None=LinearInterpolate(A=NamedRerouteUsage(), Alpha=Saturate(None=@n56)) {constB=0}) {name=StaticImage, nodeColor=<rgba 1 0 0.7969 1>}
  n57 = Divide(
   A=Subtract(A=Time(), B=ScalarParameter() {group=Reactivity, parameterName=StartTime})
   B=ScalarParameter() {defaultValue=7, group=Reactivity, parameterName=PlayTime}
  )
  n58 = ScalarParameter() {group=Reactivity, parameterName=IsInFrontEnd}
  NamedRerouteDeclaration(None=LinearInterpolate(A=Saturate(None=@n57), B=ScalarParameter() {group=Reactivity, parameterName=FrontendAlpha}, Alpha=@n58)) {name=AnimationPhase, nodeColor=<rgba 0.8047 1 0 1>}
  NamedRerouteDeclaration(None=@n33) {name=PhaseSwitch01, nodeColor=<rgba 0 1 0.1797 1>}
  NamedRerouteDeclaration(None=@n58) {name=Frontend, nodeColor=<rgba 1 0 0.4687 1>}
  NamedRerouteDeclaration(None=@n23) {name=SubtleNoiseAlpha, nodeColor=<rgba 1 0.5859 0 1>}
  NamedRerouteDeclaration(None=TextureCoordinate() {coordinateIndex=1}) {name=UVs, nodeColor=<rgba 1 0.4453 0 1>}
  n59 = NamedRerouteUsage()
  n60 = VectorParameter() {group=TV Glitches, parameterName=BlueLine, defaultValue=<rgba 0.1017 0.4775 0.8365 1>}
  n61 = ScalarParameter() {defaultValue=0.1, group=TV Glitches, parameterName=SpeedX05}
  n62 = Constant() {r=3}
  n63 = TextureObjectParameter() {
   group=TV Glitches
   parameterName=NoiseMask
   samplerType=SAMPLERTYPE_Masks
   texture=/BRCosmetics/Accessories/FORT_Backpacks/Backpack_M_MED_BisonDrain/Textures/T_BisonDrain_NoiseMask
  }
  n64 = TextureSample(UVs=Panner(Coordinate=@n3, Speed=AppendVector(A=@n61, B=@n62)), Tex=@n63) {samplerType=SAMPLERTYPE_Masks, texture=&T_BisonDrain_LineMask}
  NamedRerouteDeclaration(None=LinearInterpolate(A=LinearInterpolate(A=@n59, B=@n60.RGB, Alpha=@n64.G), B=Constant3Vector() {constant=<rgba 0.0698 0.0698 0.0698 0>}, Alpha=@n64.R) {constB=0}) {name=BackGroundImage, nodeColor=<rgba 0.4531 1 0 1>}
  NamedRerouteDeclaration(None=@n16) {name=NoiseTint, nodeColor=<rgba 0.8281 0 1 1>}
  NamedRerouteDeclaration(None=AppendVector(A=Constant(), B=@n64.B)) {name=UV_Distort, nodeColor=<rgba 1 0.2578 0 1>}
 ]}
